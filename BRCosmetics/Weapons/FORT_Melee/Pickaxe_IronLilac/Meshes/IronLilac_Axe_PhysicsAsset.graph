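# Physics asset for the IronLilac pickaxe mesh, a two-handed melee weapon cosmetic.
# It holds a single capsule body on the Melee_TwoHanded_Handle bone, offset and rotated onto the mesh.
# It has no constraints, so the weapon collides as one rigid shape with no jointed parts.
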PhysicsAsset /BRCosmetics/Weapons/FORT_Melee/Pickaxe_IronLilac/Meshes/IronLilac_Axe_PhysicsAsset
bodies {Melee_TwoHanded_Handle [{shapeType=Capsule, length=28.3533, radius=24.5883, center=<-0.3114 -12.9588 11.6327>, rotation=<rot -1.662 -90 0.0212>}]}
constraints []
